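; Physics-simulated BatFridge prop, built on physics tree log logic, that floats, knocks players around and reacts to vehicles.
; Server hits launch players by tiered impact speed, suppressed in water and delayed after riding; other hits drive surface-aware impact FX.
; Vehicle damage sets a capped velocity with a cooldown.
; Buoyancy events and a timer feed water-surface data to ripple FX, while riding and destruction report analytics.
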(event EventEndPlay
  (params (EndPlayReason :type EEndPlayReason%20Enum :default "Destroyed"))
  ;[1 1 1] End play
  (bind _returnvalue (BuildingActor|PhysicsSimulationSupport|GetPhysicsObjectComponent self))
  (bind (_numberofcachedplayer _onboardlocation _offboardlocation _sessionlength) (PhysicsObject|GetCachedRidingPlayersInfo _returnvalue true))
  (FortPhysicsAnalytics|FireTelemetryEventonPhysicsRidingActorLifeEnd self _numberofcachedplayer _onboardlocation _offboardlocation _sessionlength))

(event EventBeginPlay
  (bind _staticmeshcomponent (Variables|Debug|GetStaticMeshComponent))
  ;[1 1 1] Begin play
  (|CallParentFunction)
  (CallFunction|InitializePhysicsComponent)
  (Collision|AssignOnComponentHit (Variables|Debug|GetStaticMeshComponent) (EventDispatchers|CreateEvent :!event Custom|OnComponentHit_Event_0))
  (bind _output_get (Variables|Default|SetBuoyancyComponent (Actor|GetComponentbyClass "/Script/Water.BuoyancyComponent")))
  (Cosmetic|BindEventtoOnEnteredWaterDelegate _output_get (EventDispatchers|CreateEvent :!event Custom|OnPontoonEnteredWater))
  (Cosmetic|BindEventtoOnExitedWaterDelegate _output_get (EventDispatchers|CreateEvent :!event Custom|OnPontoonExitedWater))
  (bind _returnvalue (Utilities|Time|SetTimerbyEvent (EventDispatchers|CreateEvent :!event Custom|Update%20Water%20FX) 0.45 true))
  (Variables|Default|SetWaterSurfaceInfoTimer _returnvalue)
  (bind _outactors (Actor|GetAllActorsOfClass "/Game/Athena/Apollo/Water/Blueprints/BP_FluidSim_FN.BP_FluidSim_FN_C"))
  (Variables|Default|SetFluidSim (Utilities|Array|Get(acopy) _outactors))
  (Niagara|SpawnSystemAttached "/Game/Environments/Apollo/Foliage/PhysicsTree/FX/NS_Physics_Tree_Section_Crack_Burst.NS_Physics_Tree_Section_Crack_Burst" _staticmeshcomponent "None" "0, 0,-768.000000" "0, 0, 0" "KeepRelativeOffset" true)
  (bind _returnvalue_2 (Niagara|SpawnSystemAttached "/Game/Environments/Apollo/Foliage/PhysicsTree/FX/NS_Physics_Tree_Section_Crack_Burst.NS_Physics_Tree_Section_Crack_Burst" _staticmeshcomponent "None" "0, 0,768.000000" "0, 0, 0" "KeepRelativeOffset" true))
  (Niagara|SetNiagaraVariable(Bool) _returnvalue_2 "PlayFallingWoodBits" false)
  (Niagara|SpawnSystemAttached "/Game/Environments/Apollo/Foliage/PhysicsTree/FX/NS_Physics_Tree_Falling_Splinters.NS_Physics_Tree_Falling_Splinters" _staticmeshcomponent "None" "0, 0, 0" "0, 0, 0" "KeepRelativeOffset" true))

(event Custom|OnComponentHit_Event_0
  (params (HitComponent :type Primitive%20Component%20Object%20Reference) (OtherActor :type Actor%20Object%20Reference) (OtherComp :type Primitive%20Component%20Object%20Reference) (NormalImpulse :type Vector) (Hit :type Hit%20Result%20Structure%20%28by%20ref%29))
  (bind _returnvalue (Transformation|GetVelocity))
  (bind _returnvalue_2 (Math|Vector|VectorLength _returnvalue))
  (bind (_minimum_met _impulse) (CallFunction|GetPlayerImpulseVelocityfromImpactVelocity _returnvalue_2))
  (bind _returnvalue_3 (Utilities|Time|GetTimeSeconds))
  (bind (_bblockinghit _binitialoverlap _time _distance _location _impactpoint _normal _impactnormal _physmat) (Collision|BreakHitResult Hit))
  (bind _nextpotentiallaunchtime (Variables|Default|GetNextPotentialLaunchTime))
  (bind _returnvalue_4 (> _returnvalue_3 _nextpotentiallaunchtime))
  (bind _ns_physics_tree_environment_impact (Variables|Default|GetNS_Physics_Tree_Environment_Impact))
  (bind _ns_physics_tree_environment_impact_2 (Variables|Default|GetNS_Physics_Tree_Environment_Impact))
  ;[1 1 1] Log hit (Bound when active)
  (if (Networking|IsServer)
    (Variables|Default|SetLastImpactNormal _impactnormal)
    (bind _returnvalue_5 (BuildingActor|PhysicsSimulationSupport|GetPhysicsObjectComponent self))
    (if (Components|Activation|IsActive _returnvalue_5)
      ;[1 1 1] Launch player
      (bind _asfort_player_pawn (Utilities|Casting|CastToFortPlayerPawn OtherActor))
      (if (== (Variables|Debug|GetStaticMeshComponent) (Pawn|Components|CharacterMovement|GetMovementBaseObject (Variables|/Script/Engine.Character|GetCharacterMovement _asfort_player_pawn)))
        ;;[1 1 1] Set a launch delay for certain interactions, like riding the log in water. This will prevent the player from accidentally being launched when performing cool water tricks and jumps
        (if _returnvalue_4
          (Variables|Default|SetNextPotentialLaunchTime (+ (Utilities|Time|GetTimeSeconds) (Variables|Default|GetLaunchDelay))))
        (elif (Buoyancy|IsinWaterBody (Variables|Default|GetBuoyancyComponent))
          _
          (elif _minimum_met
            (if _returnvalue_4
              (Character|LaunchCharacter _asfort_player_pawn (* (Math|Vector|Normalize (+ (* (Variables|Default|GetLastImpactNormal) "-1.000000,-1.000000,0.000000") (Math|Vector|MakeVector 0.0 0.0 (ScalableFloat|GetValueatLevel (Variables|Default|GetVerticalImpulseRatio))))) _impulse) true true))))))
    (elif (!= (Utilities|GetClass OtherActor) (Utilities|GetClass (Game|GetPlayerPawn 0)))
      (Transformation|SetWorldTransform _ns_physics_tree_environment_impact_2 (Math|Transform|MakeTransform _location (Math|Conversions|RotationFromXVector _impactnormal)) false false)
      (if (not (Buoyancy|IsinWaterBody (Variables|Default|GetBuoyancyComponent)))
        (bind _returnvalue_6 (BuildingActor|PhysicsSimulationSupport|GetPhysicsObjectComponent self))
        (if (not (<= (Math|Vector|VectorLength (Variables|/Script/FortniteGame.FortPhysicsObjectComponent|GetLinearVelocity _returnvalue_6)) 8.0))
          (Niagara|SetNiagaraVariable(Float) _ns_physics_tree_environment_impact "Scale" (Math|Float|Clamp(Float) (* (Math|Vector|VectorLength NormalImpulse) 2e-06) 0.45))
          (Components|Activation|Activate _ns_physics_tree_environment_impact false)
          ;[1 1 1] Physical Surface
          (Niagara|SetNiagaraVariable(Int32) _ns_physics_tree_environment_impact_2 "PhysicalSurfaceID" (Math|Conversions|ToInteger(Byte) (Variables|/Script/PhysicsCore.PhysicalMaterial|GetSurfaceType _physmat)))))
      (Components|Activation|Deactivate _ns_physics_tree_environment_impact))))

(event EventOnDamageServer
  (params (Damage :type Float%20%28single-precision%29) (DamageTags :type Gameplay%20Tag%20Container%20Structure%20%28by%20ref%29) (Momentum :type Vector :default "0, 0, 0") (HitInfo :type Hit%20Result%20Structure%20%28by%20ref%29) (InstigatedBy :type Controller%20Object%20Reference) (DamageCauser :type Actor%20Object%20Reference) (EffectContext :type Gameplay%20Effect%20Context%20Handle%20Structure))
  ;[1 1 1] Donut damaged
  ;;[0 0 0 0.5] Additional processing for forces applied by vehicle damage. Vehicle damage GE is configured in PhysicsObjectPreset_BatFridge
  (CallFunction|HandleVehicleDamageForce HitInfo DamageCauser))

(event Custom|Update%20Water%20FX
  ;[1 1 1] Update Water Body Surface Info to Niagara System
  (if (Buoyancy|IsOverlappingWaterBody (Variables|Default|GetBuoyancyComponent))
    (bind (_outwaterplanelocation _outwaterplanenormal _outwatersurfaceposition _outwaterdepth _outwaterbodyidx) (Buoyancy|GetLastWaterSurfaceInfo (Variables|Default|GetBuoyancyComponent)))
    (Variables|FX|SetWaterFxPlaneLocation _outwatersurfaceposition)
    (Variables|FX|SetWaterFxPlaneNormal _outwaterplanenormal)
    (Variables|FX|SetWaterFxPlaneDepth _outwaterdepth)
    (Variables|FX|SetWaterFxIndex _outwaterbodyidx)
    (CallFunction|AdjustWaterSettings)
    (else
      (CallFunction|AdjustWaterSettings)
      (Utilities|Time|PauseTimerbyHandle (Variables|Default|GetWaterSurfaceInfoTimer)))))

(event EventOnDeathPlayEffects
  (params (Damage :type Float%20%28single-precision%29) (DamageTags :type Gameplay%20Tag%20Container%20Structure%20%28by%20ref%29) (Momentum :type Vector :default "0, 0, 0") (HitInfo :type Hit%20Result%20Structure%20%28by%20ref%29) (InstigatedBy :type Fort%20Pawn%20Object%20Reference) (DamageCauser :type Actor%20Object%20Reference) (EffectContext :type Gameplay%20Effect%20Context%20Handle%20Structure))
  ;[1 1 1] Donut damaged
  (Rendering|SetVisibility (Variables|Debug|GetStaticMeshComponent) false false)
  (if (Variables|Default|GetCanPlayDeathEffects)
    (Variables|Default|SetTreeDestructionTransform (Transformation|GetWorldTransform (Variables|Debug|GetStaticMeshComponent)))
    (SoundIndicator|StopSoundIndicatorLooping (Variables|Default|GetFortSoundIndicator))))

(event EventOnDeathServer
  (params (Damage :type Float%20%28single-precision%29) (DamageTags :type Gameplay%20Tag%20Container%20Structure%20%28by%20ref%29) (Momentum :type Vector :default "0, 0, 0") (HitInfo :type Hit%20Result%20Structure%20%28by%20ref%29) (InstigatedBy :type Controller%20Object%20Reference) (DamageCauser :type Actor%20Object%20Reference) (EffectContext :type Gameplay%20Effect%20Context%20Handle%20Structure))
  ;[1 1 1] Death Analytics
  (bind (_returnvalue _success) (CallFunction|GetAnalyticsReporterActor))
  (if _success
    (Class|BPhysicsTreeReporter|ReportLogDestroyed _returnvalue (Utilities|GetObjectName self) (Transformation|GetActorLocation self))))

(event Custom|AdjustWaterSettings
  (Networking|CanExecuteCosmeticEvents
    (:True
      (for _array_element (Variables|Default|GetFxSystemArray)
        (if (Buoyancy|IsOverlappingWaterBody (Variables|Default|GetBuoyancyComponent))
          (Components|Activation|Activate _array_element false)
          (Utilities|Time|UnpauseTimerbyHandle (Variables|Default|GetWaterSurfaceInfoTimer))
          (Effects|Components|ParticleSystem|SetVectorParameter _array_element "WaterPlaneLocation" (Variables|FX|GetWaterFxPlaneLocation))
          (Effects|Components|ParticleSystem|SetFloatParameter _array_element "WaterDepth" (Variables|FX|GetWaterFxPlaneDepth))
          (Effects|Components|ParticleSystem|SetFloatParameter _array_element "WaterBodyIndex" (Math|Conversions|ToFloat(Integer) (Variables|FX|GetWaterFxIndex)))
          (Effects|Components|ParticleSystem|SetVectorParameter _array_element "WaterPlaneNormal" (Variables|FX|GetWaterFxPlaneNormal))
          (else
            (Components|Activation|Deactivate _array_element)))))))

(event Custom|OnPontoonEnteredWater
  (params (Pontoon_CenterSocket :type Name) (Pontoon_RelativeLocation :type Vector) (Pontoon_Radius :type Float%20%28single-precision%29) (Pontoon_bFXEnabled :type Boolean) (Pontoon_LocalForce :type Vector) (Pontoon_CenterLocation :type Vector) (Pontoon_SocketRotation :type Quat%20Structure) (Pontoon_Offset :type Vector) (Pontoon_WaterHeight :type Float%20%28single-precision%29) (Pontoon_WaterDepth :type Float%20%28single-precision%29) (Pontoon_ImmersionDepth :type Float%20%28single-precision%29) (Pontoon_WaterPlaneLocation :type Vector) (Pontoon_WaterPlaneNormal :type Vector) (Pontoon_WaterSurfacePosition :type Vector) (Pontoon_WaterVelocity :type Vector) (Pontoon_WaterBodyIndex :type Integer) (Pontoon_bIsInWater :type Boolean) (Pontoon_CurrentWaterBodyComponent :type Water%20Body%20Component%20Object%20Reference))
  ;[1 1 1] Overlap Events for Water Effects
  (Networking|CanExecuteCosmeticEvents
    (:True
      (Utilities|IsValid (Variables|FX|GetWaterFxPhysicsTreeMiddle)
        (:Is%20Valid
          (CallFunction|AdjustWaterSettings))
        (:Is%20Not%20Valid
          (bind _returnvalue (Effects|SpawnFXSystemAttached :FXSystemAsset "/Game/Environments/Apollo/Foliage/PhysicsTree/FX/NS_Physics_Tree_InWater_Ripples.NS_Physics_Tree_InWater_Ripples" :AttachToComponent (Variables|Default|GetSM_WildEstate_FoodStatueFood_A) :AttachPointName "None" :Scale "1.000000,1.000000,1.000000" :LocationType "SnapToTarget" :bAutoDestroy true :PoolingMethod "None" :bAutoActivate true))
          (bind _output_get (Variables|FX|SetWaterFxPhysicsTreeMiddle _returnvalue))
          (Utilities|Array|Add (Variables|Default|GetFxSystemArray) _output_get))))))

(event Custom|OnPontoonExitedWater
  (params (Pontoon_CenterSocket :type Name) (Pontoon_RelativeLocation :type Vector) (Pontoon_Radius :type Float%20%28single-precision%29) (Pontoon_bFXEnabled :type Boolean) (Pontoon_LocalForce :type Vector) (Pontoon_CenterLocation :type Vector) (Pontoon_SocketRotation :type Quat%20Structure) (Pontoon_Offset :type Vector) (Pontoon_WaterHeight :type Float%20%28single-precision%29) (Pontoon_WaterDepth :type Float%20%28single-precision%29) (Pontoon_ImmersionDepth :type Float%20%28single-precision%29) (Pontoon_WaterPlaneLocation :type Vector) (Pontoon_WaterPlaneNormal :type Vector) (Pontoon_WaterSurfacePosition :type Vector) (Pontoon_WaterVelocity :type Vector) (Pontoon_WaterBodyIndex :type Integer) (Pontoon_bIsInWater :type Boolean) (Pontoon_CurrentWaterBodyComponent :type Water%20Body%20Component%20Object%20Reference))
  ;[1 1 1] Overlap Events for Water Effects
  (Networking|CanExecuteCosmeticEvents))

(fn CalculateSplitOffset
  (params (bSmallSide :type Boolean) (SplitPercentage :type Float%20%28double-precision%29))
  (if bSmallSide
    (return (* (- SplitPercentage (/ SplitPercentage 2.0)) (Variables|Default|GetLogLength)))
    (else
      (return (* (/ (- (Variables|Default|GetLogLength) (* (Variables|Default|GetLogLength) SplitPercentage)) 2.0) -1.0)))))

(fn CalculateSplitScale
  (params (bSmallSide :type Boolean) (SplitPercentage :type Float%20%28double-precision%29))
  (bind _staticmeshcomponent (Variables|Debug|GetStaticMeshComponent))
  (bind (_returnvalue_location _returnvalue_rotation _returnvalue_scale) (Math|Transform|BreakTransform (Transformation|GetWorldTransform _staticmeshcomponent)))
  (bind (_returnvalue_scale_x _returnvalue_scale_y _returnvalue_scale_z) (Math|Vector|BreakVector _returnvalue_scale))
  (if bSmallSide
    (return (* _returnvalue_scale_z (- 1.0 SplitPercentage)))
    (else
      (return (* _returnvalue_scale_z SplitPercentage)))))

(fn GetAnalyticsReporterActor
  (bind _returnvalue (Actor|GetActorOfClass "/Game/Athena/Items/PhysicsActors/PhysicsTreeLog/B_PhysicsTree_Reporter.B_PhysicsTree_Reporter_C"))
  (Utilities|IsValid _returnvalue
    (:Is%20Valid
      (return _returnvalue true))
    (:Is%20Not%20Valid
      (if (Networking|IsServer)
        (bind _returnvalue_2 (Game|SpawnActorfromClass "/Game/Athena/Items/PhysicsActors/PhysicsTreeLog/B_PhysicsTree_Reporter.B_PhysicsTree_Reporter_C" (Math|Transform|MakeTransform "0, 0, 0") "AlwaysSpawn" "OverrideRootScale" 0 0))
        (return _returnvalue_2 true)
        (else
          (return 0 false))))) ; Already have a valid reporter actor?
)

(fn GetPlayerImpulseVelocityFromImpactVelocity
  (params (Impact :type Float%20%28double-precision%29))
  (bind (_arrayindex _arrayelement) (Utilities|Array|ReverseforEachLoop (Variables|Default|GetPlayerImpactTiers))
    (:LoopBody
      (if (>= (Variables|GetImpact) (ScalableFloat|GetValueatLevel _arrayelement))
        (LocalVariables|SetReturnImpulseVelocity (ScalableFloat|GetValueatLevel (Utilities|Array|Get(acopy) (Variables|Default|GetPlayerImpulseTiers) _arrayindex)))))
    (:Completed
      (return (>= (LocalVariables|GetReturnImpulseVelocity) 0) (LocalVariables|GetReturnImpulseVelocity)))))

(fn GetTreeDestructionTransform
  (return (Variables|Default|GetTreeDestructionTransform)))

(fn HandleVehicleDamageForce
  (params (HitInfo :type Hit%20Result%20Structure) (HitActor :type Actor%20Object%20Reference))
  (bind (_bblockinghit _binitialoverlap _time _distance _location) (Collision|BreakHitResult (Variables|GetHitInfo)))
  (Networking|SwitchHasAuthority
    (:Authority
      (if (CallFunction|ShouldApplyVehicleForce)
        (bind _asfort_athena_vehicle (Utilities|Casting|CastToFortAthenaVehicle (Variables|GetHitActor)))
        (Variables|PhysicsInteraction|SetLastVehicleHitTimestamp (Utilities|Time|GetGameTimeinSeconds))
        (FortPhysicsSimulationLibrary|SetPhysicsLinearVelocity self (* (Math|Float|Min(Float) (* (Math|Vector|VectorLength (Vehicle|GetVehicleVelocity _asfort_athena_vehicle)) (ScalableFloat|GetValueatLevel (Variables|PhysicsInteraction|GetVehicleForceMultiplier))) (ScalableFloat|GetValueatLevel (Variables|PhysicsInteraction|GetVehicleForceLimit))) (Math|Vector|Normalize (- _location (Transformation|GetActorLocation _asfort_athena_vehicle)))) true "None" _asfort_athena_vehicle (struct :MaxSimSizeAllowed "Large" :bRejectEmbeddedInLandscape true))))))

(fn InitializePhysicsComponent
  (bind _returnvalue (BuildingActor|PhysicsSimulationSupport|GetPhysicsObjectComponent self))
  (Variables|/Script/FortniteGame.FortPhysicsObjectComponent|SetMaxLinearVelocityOverride (ScalableFloat|GetValueatLevel (Variables|PhysicsInteraction|GetPhysicsMaxLinearVelocity)) _returnvalue))

(fn OnRep_BreakTreeDamage
  (bind _staticmeshcomponent (Variables|Debug|GetStaticMeshComponent))
  (bind _returnvalue (Transformation|GetAllSocketNames _staticmeshcomponent))
  (bind _value (Math|AnimateFloatCurve 0.0 (Variables|Default|GetBreakTreeDamage) "/Game/Athena/Items/PhysicsActors/PhysicsTreeLog/C_PhysicsTree_LogChunk_Curve_01.C_PhysicsTree_LogChunk_Curve_01"))
  (Rendering|Material|SetScalarParameterValueonMaterials (Variables|Debug|GetStaticMeshComponent) "BreakTreeDamage" (/ (Math|Conversions|ToFloat(Integer) (Math|Float|Ceil _value)) (Math|Conversions|ToFloat(Integer) (- (Utilities|Array|Length _returnvalue) 1))))
  (for _index (range (Math|Integer|Min(Integer) (Variables|Default|GetCachedDamageValue) (Math|Float|Ceil _value)) (+ (Math|Integer|Min(Integer) (Math|Float|Ceil _value) (+ (Math|Integer|Min(Integer) (Variables|Default|GetCachedDamageValue) (Math|Float|Ceil _value)) 4)) 1))
    (bind (_returnvalue_location _returnvalue_rotation) (Math|Transform|BreakTransform (Transformation|GetSocketTransform _staticmeshcomponent (Utilities|Array|Get(acopy) _returnvalue _index) "RTS_World")))
    (bind _returnvalue_2 (Effects|Components|ParticleSystem|SpawnFXAtLocation "/Game/Environments/Apollo/Foliage/PhysicsTree/FX/NS_Physics_Tree_Section_Burst.NS_Physics_Tree_Section_Burst" _returnvalue_location _returnvalue_rotation))
    (Effects|Components|ParticleSystem|SetFloatParameter _returnvalue_2 "BurstScale" (Variables|Default|GetBreakTreeDamage)))
  (Variables|Default|SetCachedDamageValue (Math|Float|Ceil _value)))

(fn ReportTreeFellingToAnalytics
  (params (FelledBy :type Actor%20Object%20Reference))
  (bind _returnvalue (Actor|GetOwner FelledBy))
  (if (Networking|IsServer)
    (Utilities|FlowControl|Sequence
      (:then_0
        (Utilities|IsValid _returnvalue
          (:Is%20Valid
            (bind _asfort_player_pawn_athena (Utilities|Casting|CastToFortPlayerPawnAthena _returnvalue)
              (:then
                (bind _asfort_player_state_athena (Utilities|Casting|CastToFortPlayerStateAthena (Variables|/Script/Engine.Pawn|GetPlayerState _asfort_player_pawn_athena)))
                (bind _returnvalue_2 (PlayerInfo|GetDebugStringforUniqueId (PlayerState|GetPlayerUniqueNetID _asfort_player_state_athena)))
                (LocalVariables|SetFeller _returnvalue_2))
              (:CastFailed
                _)))
          (:Is%20Not%20Valid
            _))
        (LocalVariables|SetFeller (Utilities|GetObjectName (Variables|GetFelledBy))))
      (:then_1
        (bind (_returnvalue_3 _success) (CallFunction|GetAnalyticsReporterActor))
        (if _success
          (Class|BPhysicsTreeReporter|ReportLogFelled _returnvalue_3 (Utilities|GetObjectName self) (LocalVariables|GetFeller) (Transformation|GetActorLocation self)))))))

(fn ShouldApplyVehicleForce
  (return (> (Utilities|Time|GetGameTimeinSeconds) (+ (Variables|PhysicsInteraction|GetLastVehicleHitTimestamp) (ScalableFloat|GetValueatLevel (Variables|PhysicsInteraction|GetVehicleHitCooldown))))))

(fn ConstructionScript
  (|CallParentFunction))
