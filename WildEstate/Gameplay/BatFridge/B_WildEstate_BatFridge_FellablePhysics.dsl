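; Fellable physics Blueprint for the WildEstate BatFridge prop, handling its death sound and fall direction.
; On BeginPlay it destroys the FoodStatueFood HLOD component, and on death it calls the parent, then spawns a sound attached at BreakLocation.
; SetFallVectorXY builds the 2D fall vector from the LaunchDirection component's forward axis, using a random bool to choose forward or reversed.
; ApplyChrome and the construction script are empty.
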(event EventBeginPlay
  (Components|DestroyComponent (Variables|Default|GetSM_WildEstate_FoodStatueFood_HLOD)))

(event EventOnDeathPlayEffects
  (params (Damage :type Float%20%28single-precision%29) (DamageTags :type Gameplay%20Tag%20Container%20Structure%20%28by%20ref%29) (Momentum :type Vector :default "0, 0, 0") (HitInfo :type Hit%20Result%20Structure%20%28by%20ref%29) (InstigatedBy :type Fort%20Pawn%20Object%20Reference) (DamageCauser :type Actor%20Object%20Reference) (EffectContext :type Gameplay%20Effect%20Context%20Handle%20Structure))
  (|CallParentFunction Damage DamageTags Momentum HitInfo InstigatedBy DamageCauser EffectContext)
  (Audio|SpawnSoundAttached (Variables|Default|GetSound) (Variables|Default|GetBreakLocation)))

(fn ApplyChrome
  (params (Target :type Actor%20Object%20Reference)))

(fn SetFallVectorXY
  (Variables|FellingResults|SetFallVector (Math|Vector2D|MakeVector2D (.x (Math|Vector|SelectVector (Math|Rotator|GetRotationXVector (Transformation|GetWorldRotation (Variables|Default|GetLaunchDirection))) (* (Math|Rotator|GetRotationXVector (Transformation|GetWorldRotation (Variables|Default|GetLaunchDirection))) -1.0) (Math|Random|RandomBool))) (.y (Math|Vector|SelectVector (Math|Rotator|GetRotationXVector (Transformation|GetWorldRotation (Variables|Default|GetLaunchDirection))) (* (Math|Rotator|GetRotationXVector (Transformation|GetWorldRotation (Variables|Default|GetLaunchDirection))) -1.0) (Math|Random|RandomBool))))))

(fn ConstructionScript)
